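# Sound cue that plays the full SummerStride back bling cosmetic sound.
# A single wave player feeds the output directly at slightly reduced volume, using a close back bling attenuation and the Backblings sound class.
# A reactive cosmetic back bling property sheet is applied, and overridden concurrency allows one instance per owner, rejecting new plays while one is active.
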SoundCue /BRCosmetics/Sounds/Backblings/SummerStride/BackBling_SummerStride_Full_Cue
root=WavePlayer_0
nodes [{id=WavePlayer_0, properties {soundWaveAssetPtr=&/Backbling_SummerStride_Full}}]
properties {
 attenuationSettings=/BRCosmetics/Sounds/Other/Attenuation/Fort_BackBling_Close_Att
 audioPropertiesSheet=/BRCosmetics/Sounds/Settings/PropertySheets/Reactive/APS_Cosmetic_Reactive_Backbling
 bOverrideConcurrency=true
 maxDistance=1111.11
 soundClassObject=/Game/Sounds/FortSoundClassesAndModes/Fort_Backblings
 volumeMultiplier=0.8
 concurrencyOverrides {bLimitToOwner=true, maxCount=1, resolutionRule=PreventNew}
}
